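# Material for Apollo's background landscape backdrop, layering cliff rock, water, beach, grass and snow.
# Two rock layers and grass height-blend using the transformed backdrop normal's Z as slope, with vertex red scaling grass transition.
# World height separates water, beach and mask-patterned grass, while vertex alpha and cliff height drive snow.
# Forward and mobile paths replace rock with flat shading; layer normal XY is added onto the backdrop normal.
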Material /Clyde/Athena/Apollo/Environments/Landscape/Background/M_BackDrop_P
graph {
 properties {
  bUseMaterialAttributes=true
  bUsedWithMeshDeformer=true
  bUsedWithSkeletalMesh=true
  pixelDepthOffsetMode=PDOM_Legacy
  usedShadingModels=Default Lit
 }
 nodes [
  n1 = TextureObjectParameter() {
   group=Dirt2
   parameterName=Specular
   samplerType=SAMPLERTYPE_Masks
   texture=/Game/Environments/Apollo/Sets/Terrain/Textures/T_Apollo_Cliff_S_01
  }
  n2 = TextureObjectParameter() {
   group=Dirt2
   parameterName=Normals
   samplerType=SAMPLERTYPE_Normal
   texture=/Game/Environments/Apollo/Sets/Terrain/Textures/T_Apollo_Cliff_N_01
  }
  n3 = MaterialFunctionCall!(
   Diffuse=TextureObjectParameter() {group=Dirt2, parameterName=Diffuse, texture=/Game/Environments/Apollo/Sets/Terrain/Textures/T_Apollo_Cliff_D_01}
   Specular=@n1
   Normal=@n2
   Nomal Value=VectorParameter() {group=LayerCliff, parameterName=Cliffs N value, sortPriority=4, defaultValue=<rgba 0.5 0.5 1 0>}.RGB
   BaseColorTone Walls
   Tile=ScalarParameter() {defaultValue=195555, parameterName=Rocks 1 Scale}
  ) {materialFunction=/Game/Environments/Apollo/VS/Materials/MF_Apollo_Cliff}
  n4 = Multiply(A=TextureCoordinate(), B=ScalarParameter() {defaultValue=10.7482, parameterName=Rocks2 Scale})
  n5 = TextureObjectParameter() {group=Dirt2, parameterName=Diffuse2, texture=/Game/Environments/Apollo/Sets/Terrain/Textures/T_Apollo_Cliff_D_02}
  n6 = TextureObjectParameter() {
   group=Dirt2
   parameterName=Specular2
   samplerType=SAMPLERTYPE_Masks
   texture=/Game/Environments/Apollo/Sets/Terrain/Textures/T_Apollo_Cliff_S_02
  }
  n7 = TextureObjectParameter() {
   group=Dirt2
   parameterName=Normals2
   samplerType=SAMPLERTYPE_Normal
   texture=/Game/Environments/Apollo/Sets/Terrain/Textures/T_Apollo_Cliff_N_02
  }
  n8 = TextureSampleParameter2D() {parameterName=Normal, samplerType=SAMPLERTYPE_Normal, texture=/Game/Athena/Apollo/Environments/Landscape/Background/BackDrop_1_N}
  n9 = ComponentMask(None=Transform(None=@n8.RGB)) {b=true}
  n10 = BlendMaterialAttributes(
   A=@n3.Result 3D
   B=SetMaterialAttributes!(
    MaterialAttributes
    Base Color=TextureSample(UVs=@n4, Tex=@n5) {samplerType=SAMPLERTYPE_Masks, texture=&T_Apollo_Cliff_S_02}.RGB
    Roughness=TextureSample(UVs=@n4, Tex=@n6) {samplerType=SAMPLERTYPE_Masks, texture=&T_Apollo_Cliff_S_02}.B
    Normal=TextureSample(UVs=@n4, Tex=@n7) {samplerType=SAMPLERTYPE_Masks, texture=&T_Apollo_Cliff_S_02}.RGB
   ) {attributeSetTypes [#1, #2, #3]}
   Alpha=MaterialFunctionCall(
    Transition Phase=ScalarParameter() {defaultValue=0.3802, parameterName=Transition Rocks}
    Height Texture=@n9
    Contrast=ScalarParameter() {parameterName=Contrast rocks}
   ) {materialFunction=&HeightLerp}.Alpha
  )
  n11 = SetMaterialAttributes!(MaterialAttributes, Base Color=Constant() {r=0.25}, Roughness) {attributeSetTypes [#1, #2]}
  n12 = ComponentMask(None=WorldPosition().XYZ) {b=true}
  n13 = Constant() {r=-5000}
  n14 = BlendMaterialAttributes(
   A=SetMaterialAttributes!(
    MaterialAttributes
    Base Color=VectorParameter() {group=Beach, parameterName=Water, defaultValue=<rgba 0.0346 0.081 0.1719 0>}.RGB
    Specular=Constant() {r=0.2}
    Roughness=Constant()
   ) {attributeSetTypes [#1, #4, #2]}
   B=SetMaterialAttributes!(
    MaterialAttributes
    Base Color=VectorParameter() {group=Beach, parameterName=BeachLine, defaultValue=<rgba 0.5781 0.5188 0.3735 0>}.RGB
    Specular=Constant() {r=0.3}
   ) {attributeSetTypes [#1, #4]}
   Alpha=Saturate(None=Subtract(A=@n12, B=@n13))
  )
  n15 = VectorParameter() {group=Grass, parameterName=Grass2, defaultValue=<rgba 0.3478 0.4531 0.0696 0>}
  n16 = VectorParameter() {group=Grass, parameterName=Grass, defaultValue=<rgba 0.1247 0.2135 0.04 0>}
  n17 = ScalarParameter() {defaultValue=0.65, group=Grass, parameterName=Grass Patern Trans}
  n18 = WorldPosition() {worldPositionShaderOffset=WPT_ExcludeAllShaderOffsets}
  n19 = TextureSample(UVs=Divide(A=ComponentMask(None=@n18.XYZ) {g=true, r=true}, B=ScalarParameter() {defaultValue=850000, parameterName=GrassDetailTexture-Size})) {
   samplerSource=SSM_Wrap_WorldGroupSettings
   samplerType=SAMPLERTYPE_Grayscale
   texture=/Game/Athena/Apollo/Environments/Landscape/Materials/Biome/CommonMaterials/T_ApolloGrass_ColorMask
  }
  n20 = ScalarParameter() {defaultValue=1, group=Grass, parameterName=Grass Pattern Contrast}
  n21 = WorldPosition()
  n22 = Divide(
   A=Subtract(A=ComponentMask(None=@n21.XYZ) {b=true}, B=ScalarParameter() {defaultValue=-5500, group=Grass, parameterName=GrassTone Height})
   B=Constant() {r=155555}
  )
  n23 = SetMaterialAttributes!(
   MaterialAttributes
   Base Color=LinearInterpolate(
    A=VectorParameter() {group=Grass, parameterName=Grass Low, defaultValue=<rgba 0.292 0.3646 0.0019 0>}.RGB
    B=MaterialFunctionCall(A=@n15.RGB, B=@n16.RGB, Transition Phase=@n17, Height Texture=@n19.B, Contrast=@n20) {materialFunction=&HeightLerp}.Results
    Alpha=Saturate(None=@n22)
   )
   Roughness=ScalarParameter() {defaultValue=0.8, group=Grass, parameterName=Grass Roughness}
  ) {attributeSetTypes [#1, #2]}
  n24 = Constant() {r=-3200}
  n25 = Saturate(None=Divide(A=Subtract(A=@n12, B=@n24), B=Constant() {r=500}))
  n26 = Multiply(A=ScalarParameter() {defaultValue=0.4911, parameterName=Transition Grass}, B=VertexColor().R)
  n27 = ScalarParameter() {defaultValue=5, parameterName=Contrast Grass}
  n28 = GetMaterialAttributes(None=BlendMaterialAttributes(
    A=ShadingPathSwitch(Default=@n10, Deferred=@n10, Forward=@n11, Mobile=@n11)
    B=BlendMaterialAttributes(A=@n14, B=@n23, Alpha=@n25)
    Alpha=MaterialFunctionCall(Transition Phase=@n26, Height Texture=@n9, Contrast=@n27) {materialFunction=&HeightLerp}.Alpha
   )) {attributeGetTypes [#1]}
  n29 = VectorParameter() {group=Snow, parameterName="Snow ", defaultValue=<rgba 0.8 0.8 0.8 0>}
  n30 = VertexColor()
  n31 = Constant() {r=-12}
  n32 = GetMaterialAttributes(None=SetMaterialAttributes!(
    MaterialAttributes=@n28.MaterialAttributes
    Base Color=MaterialFunctionCall(A=@n28.Base Color, B=@n29.RGB, Transition Phase=@n30.A, Height Texture=@n3.Height map 3D, Contrast=@n31) {materialFunction=&HeightLerp}.Results
   ) {attributeSetTypes [#1]}) {attributeGetTypes [#3]}
  n33 = Constant3Vector() {constant=<rgba 1 1 0 0>}
  n34 = TextureSampleParameter2D() {parameterName=Normal, samplerType=SAMPLERTYPE_Normal, texture=&BackDrop_1_N}
  SetMaterialAttributes!(MaterialAttributes=@n32.MaterialAttributes, Normal=Add(A=Multiply(A=@n32.Normal, B=@n33), B=@n34.RGB)) {attributeSetTypes [#3]}
 ]
}
