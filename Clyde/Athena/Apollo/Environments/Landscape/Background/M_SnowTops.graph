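# M_SnowTops is a simple snow material, apparently for background landscape tops in the Apollo environment, using the Subsurface shading model.
# It exposes parameters for roughness, specular, a light grey color and a pale cyan subsurface color, so instances can retune the look.
# The graph contains only these constant parameters, with no textures or math nodes.
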Material /Clyde/Athena/Apollo/Environments/Landscape/Background/M_SnowTops
graph {
 properties {pixelDepthOffsetMode=PDOM_Legacy, shadingModel=MSM_Subsurface, usedShadingModels=Subsurface}
 nodes [
  ScalarParameter() {defaultValue=0.7, parameterName=Roughness}
  ScalarParameter() {defaultValue=0.75, parameterName=Specular}
  VectorParameter() {parameterName=Color, defaultValue=<rgba 0.7917 0.7917 0.7917 0>}
  VectorParameter() {parameterName=SSSColor, defaultValue=<rgba 0.5364 0.8367 0.8802 0>}
 ]
}
